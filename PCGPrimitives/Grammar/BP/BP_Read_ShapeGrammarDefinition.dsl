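; Custom PCG Blueprint node "Read Shape Grammar Definition" that converts a BP_ShapeGrammarDefinition asset into PCG attribute sets for shape-grammar generation.
; For each input param data, it creates an attribute set with slot, rule and module attributes.
; Nested loops flatten slots, their rule assets and each rule's modules into separate entries, linked by Slot ID and Rule ID and distinguished by Type.
; Rule assets and module meshes are stored as soft object paths.
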(fn ExecuteWithContext
  (params (InContext :type PCGContext%20Structure%20%28by%20ref%29) (Input :type PCGData%20Collection%20Structure%20%28by%20ref%29))
  (bind (_positiononslot_57_f123cd704574b53c5d934eaceec6c482 _overrideprofilegrammar_36_4ee7ba9446dcf473b7bc3391302e7ef5 _grammar_10_4003928d40c6099dca57829f48235d57 _modulesinfos_38_0ea66fd04214ae8d9f6b83a1b57ac47b _useupvector_42_d1e94e694ccc8a3a0238d19ddfb7008b _upvector_46_14c9927646621a47516c6eb8732ef3ec _randposrange_21_4dd3a8d948c5e00d1cd838ad6f83a3df) (Utilities|Struct|BreakSTCRule (Variables|/PCGPrimitives/Grammar/BP/BP_ShapeGrammarRule.BP_ShapeGrammarRule_C|GetRule (LocalVariables|Getcurrent%20Rule%20asset))))
  (bind (_positiononslot_57_f123cd704574b53c5d934eaceec6c482_2 _overrideprofilegrammar_36_4ee7ba9446dcf473b7bc3391302e7ef5_2 _grammar_10_4003928d40c6099dca57829f48235d57_2 _modulesinfos_38_0ea66fd04214ae8d9f6b83a1b57ac47b_2 _useupvector_42_d1e94e694ccc8a3a0238d19ddfb7008b_2 _upvector_46_14c9927646621a47516c6eb8732ef3ec_2 _randposrange_21_4dd3a8d948c5e00d1cd838ad6f83a3df_2 _randrotrange_23_ad11dd064aacbe6ea2fac8ab29b52f76) (Utilities|Struct|BreakSTCRule (Variables|/PCGPrimitives/Grammar/BP/BP_ShapeGrammarRule.BP_ShapeGrammarRule_C|GetRule (LocalVariables|Getcurrent%20Rule%20asset))))
  (bind (_positiononslot_57_f123cd704574b53c5d934eaceec6c482_3 _overrideprofilegrammar_36_4ee7ba9446dcf473b7bc3391302e7ef5_3) (Utilities|Struct|BreakSTCRule (Variables|/PCGPrimitives/Grammar/BP/BP_ShapeGrammarRule.BP_ShapeGrammarRule_C|GetRule (LocalVariables|Getcurrent%20Rule%20asset))))
  (bind _current_slot_config (LocalVariables|Getcurrent%20Slot%20Config))
  (bind (_shapegrammarrules_42_0cb868434643776af184f78028144928 _symbol_24_e4469dfa4802eef57e43d29349fbc6b7 _mirror_40_8ac86c574422496d3787acab2ecdd044 _size_31_a6850f3346aa84d1d37a079b63b8c073) (Utilities|Struct|BreakSTCSlot _current_slot_config))
  (bind (_positiononslot_57_f123cd704574b53c5d934eaceec6c482_4 _overrideprofilegrammar_36_4ee7ba9446dcf473b7bc3391302e7ef5_4 _grammar_10_4003928d40c6099dca57829f48235d57_3 _modulesinfos_38_0ea66fd04214ae8d9f6b83a1b57ac47b_3 _useupvector_42_d1e94e694ccc8a3a0238d19ddfb7008b_3 _upvector_46_14c9927646621a47516c6eb8732ef3ec_3) (Utilities|Struct|BreakSTCRule (Variables|/PCGPrimitives/Grammar/BP/BP_ShapeGrammarRule.BP_ShapeGrammarRule_C|GetRule (LocalVariables|Getcurrent%20Rule%20asset))))
  (bind (_positiononslot_57_f123cd704574b53c5d934eaceec6c482_5 _overrideprofilegrammar_36_4ee7ba9446dcf473b7bc3391302e7ef5_5 _grammar_10_4003928d40c6099dca57829f48235d57_4 _modulesinfos_38_0ea66fd04214ae8d9f6b83a1b57ac47b_4 _useupvector_42_d1e94e694ccc8a3a0238d19ddfb7008b_4) (Utilities|Struct|BreakSTCRule (Variables|/PCGPrimitives/Grammar/BP/BP_ShapeGrammarRule.BP_ShapeGrammarRule_C|GetRule (LocalVariables|Getcurrent%20Rule%20asset))))
  (bind (_positiononslot_57_f123cd704574b53c5d934eaceec6c482_6 _overrideprofilegrammar_36_4ee7ba9446dcf473b7bc3391302e7ef5_6 _grammar_10_4003928d40c6099dca57829f48235d57_5 _modulesinfos_38_0ea66fd04214ae8d9f6b83a1b57ac47b_5 _useupvector_42_d1e94e694ccc8a3a0238d19ddfb7008b_5 _upvector_46_14c9927646621a47516c6eb8732ef3ec_4 _randposrange_21_4dd3a8d948c5e00d1cd838ad6f83a3df_3 _randrotrange_23_ad11dd064aacbe6ea2fac8ab29b52f76_2 _randscalerange_39_35a8f3d34adce6dd9523b7a68aa529ca _posoffset_30_a18be791408ce7b07aa48da39140fbe7 _overridescale_27_99450af14e8c8ac67e1134b0d87fe85d _spawnsplinemesh_48_a2446a5d40c500f8c62ab69df04cf61c _splinemeshanglethreshold_56_19b857f641a7fd07c727379c3e41af0a _setsplinemeshroll_52_00f53574414c67e7869456be70691314 _splinerolldegrees_55_a51b57a945178d7632154db6fb7effd4) (Utilities|Struct|BreakSTCRule (Variables|/PCGPrimitives/Grammar/BP/BP_ShapeGrammarRule.BP_ShapeGrammarRule_C|GetRule (LocalVariables|Getcurrent%20Rule%20asset))))
  (bind (_mesh_69_abc2f260431a21daf912f49032642918 _symbol_6_f9fb29564dea8bef1ea1dd8aa30c03e3 _scalable_68_5f7bc934476923af3c7bbcb6f7239ecf _guesssize_79_06d919a9451b09f481d71f9db1b48c89 _size_67_aa4253b1485ed7a89506fca171ac20ba _guesspivotoffset_80_dd4df4ef457ae0dcc42e89abe8bbb6c2 _pivotoffset_28_063b0b4741dd9478c20e8e83a4dff20a _meshforwardaxis_35_62b04dfb4a4ba73267f123a1d53ac25f _flipforwardaxis_30_2ac382c0416cbac8fe032f998f3994de _debugcolor_83_34bc0d8c4720597367b304b35aafee3f) (Utilities|Struct|BreakSTCModule (LocalVariables|GetCurrent%20Grammar%20Module%20def)))
  (bind _current_grammar_module_def (LocalVariables|GetCurrent%20Grammar%20Module%20def))
  (bind _mesh_69_abc2f260431a21daf912f49032642918_2 (Utilities|Struct|BreakSTCModule _current_grammar_module_def))
  (bind (_mesh_69_abc2f260431a21daf912f49032642918_3 _symbol_6_f9fb29564dea8bef1ea1dd8aa30c03e3_2) (Utilities|Struct|BreakSTCModule (LocalVariables|GetCurrent%20Grammar%20Module%20def)))
  (bind (_mesh_69_abc2f260431a21daf912f49032642918_4 _symbol_6_f9fb29564dea8bef1ea1dd8aa30c03e3_3 _scalable_68_5f7bc934476923af3c7bbcb6f7239ecf_2) (Utilities|Struct|BreakSTCModule (LocalVariables|GetCurrent%20Grammar%20Module%20def)))
  (bind (_mesh_69_abc2f260431a21daf912f49032642918_5 _symbol_6_f9fb29564dea8bef1ea1dd8aa30c03e3_4 _scalable_68_5f7bc934476923af3c7bbcb6f7239ecf_3 _guesssize_79_06d919a9451b09f481d71f9db1b48c89_2) (Utilities|Struct|BreakSTCModule (LocalVariables|GetCurrent%20Grammar%20Module%20def)))
  (bind (_positiononslot_57_f123cd704574b53c5d934eaceec6c482_7 _overrideprofilegrammar_36_4ee7ba9446dcf473b7bc3391302e7ef5_7 _grammar_10_4003928d40c6099dca57829f48235d57_6 _modulesinfos_38_0ea66fd04214ae8d9f6b83a1b57ac47b_6 _useupvector_42_d1e94e694ccc8a3a0238d19ddfb7008b_6 _upvector_46_14c9927646621a47516c6eb8732ef3ec_5 _randposrange_21_4dd3a8d948c5e00d1cd838ad6f83a3df_4 _randrotrange_23_ad11dd064aacbe6ea2fac8ab29b52f76_3 _randscalerange_39_35a8f3d34adce6dd9523b7a68aa529ca_2) (Utilities|Struct|BreakSTCRule (Variables|/PCGPrimitives/Grammar/BP/BP_ShapeGrammarRule.BP_ShapeGrammarRule_C|GetRule (LocalVariables|Getcurrent%20Rule%20asset))))
  (bind (_mesh_69_abc2f260431a21daf912f49032642918_6 _symbol_6_f9fb29564dea8bef1ea1dd8aa30c03e3_5 _scalable_68_5f7bc934476923af3c7bbcb6f7239ecf_4 _guesssize_79_06d919a9451b09f481d71f9db1b48c89_3 _size_67_aa4253b1485ed7a89506fca171ac20ba_2 _guesspivotoffset_80_dd4df4ef457ae0dcc42e89abe8bbb6c2_2 _pivotoffset_28_063b0b4741dd9478c20e8e83a4dff20a_2 _meshforwardaxis_35_62b04dfb4a4ba73267f123a1d53ac25f_2 _flipforwardaxis_30_2ac382c0416cbac8fe032f998f3994de_2) (Utilities|Struct|BreakSTCModule (LocalVariables|GetCurrent%20Grammar%20Module%20def)))
  (bind (_mesh_69_abc2f260431a21daf912f49032642918_7 _symbol_6_f9fb29564dea8bef1ea1dd8aa30c03e3_6 _scalable_68_5f7bc934476923af3c7bbcb6f7239ecf_5 _guesssize_79_06d919a9451b09f481d71f9db1b48c89_4 _size_67_aa4253b1485ed7a89506fca171ac20ba_3 _guesspivotoffset_80_dd4df4ef457ae0dcc42e89abe8bbb6c2_3) (Utilities|Struct|BreakSTCModule (LocalVariables|GetCurrent%20Grammar%20Module%20def)))
  (bind (_mesh_69_abc2f260431a21daf912f49032642918_8 _symbol_6_f9fb29564dea8bef1ea1dd8aa30c03e3_7 _scalable_68_5f7bc934476923af3c7bbcb6f7239ecf_6 _guesssize_79_06d919a9451b09f481d71f9db1b48c89_5 _size_67_aa4253b1485ed7a89506fca171ac20ba_4) (Utilities|Struct|BreakSTCModule (LocalVariables|GetCurrent%20Grammar%20Module%20def)))
  (bind (_mesh_69_abc2f260431a21daf912f49032642918_9 _symbol_6_f9fb29564dea8bef1ea1dd8aa30c03e3_8 _scalable_68_5f7bc934476923af3c7bbcb6f7239ecf_7 _guesssize_79_06d919a9451b09f481d71f9db1b48c89_6 _size_67_aa4253b1485ed7a89506fca171ac20ba_5 _guesspivotoffset_80_dd4df4ef457ae0dcc42e89abe8bbb6c2_4 _pivotoffset_28_063b0b4741dd9478c20e8e83a4dff20a_3) (Utilities|Struct|BreakSTCModule (LocalVariables|GetCurrent%20Grammar%20Module%20def)))
  (bind (_positiononslot_57_f123cd704574b53c5d934eaceec6c482_8 _overrideprofilegrammar_36_4ee7ba9446dcf473b7bc3391302e7ef5_8 _grammar_10_4003928d40c6099dca57829f48235d57_7 _modulesinfos_38_0ea66fd04214ae8d9f6b83a1b57ac47b_7 _useupvector_42_d1e94e694ccc8a3a0238d19ddfb7008b_7 _upvector_46_14c9927646621a47516c6eb8732ef3ec_6 _randposrange_21_4dd3a8d948c5e00d1cd838ad6f83a3df_5 _randrotrange_23_ad11dd064aacbe6ea2fac8ab29b52f76_4 _randscalerange_39_35a8f3d34adce6dd9523b7a68aa529ca_3 _posoffset_30_a18be791408ce7b07aa48da39140fbe7_2 _overridescale_27_99450af14e8c8ac67e1134b0d87fe85d_2 _spawnsplinemesh_48_a2446a5d40c500f8c62ab69df04cf61c_2) (Utilities|Struct|BreakSTCRule (Variables|/PCGPrimitives/Grammar/BP/BP_ShapeGrammarRule.BP_ShapeGrammarRule_C|GetRule (LocalVariables|Getcurrent%20Rule%20asset))))
  (bind (_positiononslot_57_f123cd704574b53c5d934eaceec6c482_9 _overrideprofilegrammar_36_4ee7ba9446dcf473b7bc3391302e7ef5_9 _grammar_10_4003928d40c6099dca57829f48235d57_8 _modulesinfos_38_0ea66fd04214ae8d9f6b83a1b57ac47b_8 _useupvector_42_d1e94e694ccc8a3a0238d19ddfb7008b_8 _upvector_46_14c9927646621a47516c6eb8732ef3ec_7 _randposrange_21_4dd3a8d948c5e00d1cd838ad6f83a3df_6 _randrotrange_23_ad11dd064aacbe6ea2fac8ab29b52f76_5 _randscalerange_39_35a8f3d34adce6dd9523b7a68aa529ca_4 _posoffset_30_a18be791408ce7b07aa48da39140fbe7_3 _overridescale_27_99450af14e8c8ac67e1134b0d87fe85d_3 _spawnsplinemesh_48_a2446a5d40c500f8c62ab69df04cf61c_3 _splinemeshanglethreshold_56_19b857f641a7fd07c727379c3e41af0a_2) (Utilities|Struct|BreakSTCRule (Variables|/PCGPrimitives/Grammar/BP/BP_ShapeGrammarRule.BP_ShapeGrammarRule_C|GetRule (LocalVariables|Getcurrent%20Rule%20asset))))
  (bind (_shapegrammarrules_42_0cb868434643776af184f78028144928_2 _symbol_24_e4469dfa4802eef57e43d29349fbc6b7_2 _mirror_40_8ac86c574422496d3787acab2ecdd044_2 _size_31_a6850f3346aa84d1d37a079b63b8c073_2 _scalable_35_e032e01047c16dd30ae042a7e8ee0048 _offset_33_5f8333ed4b9d55156e925ab9e6981a83 _iscorner_37_537b72c44f41655dfbc568871556a218) (Utilities|Struct|BreakSTCSlot (LocalVariables|Getcurrent%20Slot%20Config)))
  (bind (_b _g _r _a) (Utilities|Struct|BreakColor _debugcolor_83_34bc0d8c4720597367b304b35aafee3f))
  (bind (_mesh_69_abc2f260431a21daf912f49032642918_10 _symbol_6_f9fb29564dea8bef1ea1dd8aa30c03e3_9 _scalable_68_5f7bc934476923af3c7bbcb6f7239ecf_8 _guesssize_79_06d919a9451b09f481d71f9db1b48c89_7 _size_67_aa4253b1485ed7a89506fca171ac20ba_6 _guesspivotoffset_80_dd4df4ef457ae0dcc42e89abe8bbb6c2_5 _pivotoffset_28_063b0b4741dd9478c20e8e83a4dff20a_4 _meshforwardaxis_35_62b04dfb4a4ba73267f123a1d53ac25f_3) (Utilities|Struct|BreakSTCModule (LocalVariables|GetCurrent%20Grammar%20Module%20def)))
  (bind _current_rule_asset (LocalVariables|Getcurrent%20Rule%20asset))
  (bind _rule (Variables|/PCGPrimitives/Grammar/BP/BP_ShapeGrammarRule.BP_ShapeGrammarRule_C|GetRule _current_rule_asset))
  (bind (_positiononslot_57_f123cd704574b53c5d934eaceec6c482_10 _overrideprofilegrammar_36_4ee7ba9446dcf473b7bc3391302e7ef5_10 _grammar_10_4003928d40c6099dca57829f48235d57_9) (Utilities|Struct|BreakSTCRule _rule))
  (bind (_positiononslot_57_f123cd704574b53c5d934eaceec6c482_11 _overrideprofilegrammar_36_4ee7ba9446dcf473b7bc3391302e7ef5_11 _grammar_10_4003928d40c6099dca57829f48235d57_10 _modulesinfos_38_0ea66fd04214ae8d9f6b83a1b57ac47b_9 _useupvector_42_d1e94e694ccc8a3a0238d19ddfb7008b_9 _upvector_46_14c9927646621a47516c6eb8732ef3ec_8 _randposrange_21_4dd3a8d948c5e00d1cd838ad6f83a3df_7 _randrotrange_23_ad11dd064aacbe6ea2fac8ab29b52f76_6 _randscalerange_39_35a8f3d34adce6dd9523b7a68aa529ca_5 _posoffset_30_a18be791408ce7b07aa48da39140fbe7_4 _overridescale_27_99450af14e8c8ac67e1134b0d87fe85d_4) (Utilities|Struct|BreakSTCRule (Variables|/PCGPrimitives/Grammar/BP/BP_ShapeGrammarRule.BP_ShapeGrammarRule_C|GetRule (LocalVariables|Getcurrent%20Rule%20asset))))
  (bind (_shapegrammarrules_42_0cb868434643776af184f78028144928_3 _symbol_24_e4469dfa4802eef57e43d29349fbc6b7_3 _mirror_40_8ac86c574422496d3787acab2ecdd044_3 _size_31_a6850f3346aa84d1d37a079b63b8c073_3 _scalable_35_e032e01047c16dd30ae042a7e8ee0048_2 _offset_33_5f8333ed4b9d55156e925ab9e6981a83_2) (Utilities|Struct|BreakSTCSlot (LocalVariables|Getcurrent%20Slot%20Config)))
  (bind (_positiononslot_57_f123cd704574b53c5d934eaceec6c482_12 _overrideprofilegrammar_36_4ee7ba9446dcf473b7bc3391302e7ef5_12 _grammar_10_4003928d40c6099dca57829f48235d57_11 _modulesinfos_38_0ea66fd04214ae8d9f6b83a1b57ac47b_10) (Utilities|Struct|BreakSTCRule (Variables|/PCGPrimitives/Grammar/BP/BP_ShapeGrammarRule.BP_ShapeGrammarRule_C|GetRule (LocalVariables|Getcurrent%20Rule%20asset))))
  (bind (_shapegrammarrules_42_0cb868434643776af184f78028144928_4 _symbol_24_e4469dfa4802eef57e43d29349fbc6b7_4 _mirror_40_8ac86c574422496d3787acab2ecdd044_4 _size_31_a6850f3346aa84d1d37a079b63b8c073_4 _scalable_35_e032e01047c16dd30ae042a7e8ee0048_3) (Utilities|Struct|BreakSTCSlot (LocalVariables|Getcurrent%20Slot%20Config)))
  (bind (_positiononslot_57_f123cd704574b53c5d934eaceec6c482_13 _overrideprofilegrammar_36_4ee7ba9446dcf473b7bc3391302e7ef5_13 _grammar_10_4003928d40c6099dca57829f48235d57_12 _modulesinfos_38_0ea66fd04214ae8d9f6b83a1b57ac47b_11 _useupvector_42_d1e94e694ccc8a3a0238d19ddfb7008b_10 _upvector_46_14c9927646621a47516c6eb8732ef3ec_9 _randposrange_21_4dd3a8d948c5e00d1cd838ad6f83a3df_8 _randrotrange_23_ad11dd064aacbe6ea2fac8ab29b52f76_7 _randscalerange_39_35a8f3d34adce6dd9523b7a68aa529ca_6 _posoffset_30_a18be791408ce7b07aa48da39140fbe7_5) (Utilities|Struct|BreakSTCRule (Variables|/PCGPrimitives/Grammar/BP/BP_ShapeGrammarRule.BP_ShapeGrammarRule_C|GetRule (LocalVariables|Getcurrent%20Rule%20asset))))
  (bind (_positiononslot_57_f123cd704574b53c5d934eaceec6c482_14 _overrideprofilegrammar_36_4ee7ba9446dcf473b7bc3391302e7ef5_14 _grammar_10_4003928d40c6099dca57829f48235d57_13 _modulesinfos_38_0ea66fd04214ae8d9f6b83a1b57ac47b_12 _useupvector_42_d1e94e694ccc8a3a0238d19ddfb7008b_11 _upvector_46_14c9927646621a47516c6eb8732ef3ec_10 _randposrange_21_4dd3a8d948c5e00d1cd838ad6f83a3df_9 _randrotrange_23_ad11dd064aacbe6ea2fac8ab29b52f76_8 _randscalerange_39_35a8f3d34adce6dd9523b7a68aa529ca_7 _posoffset_30_a18be791408ce7b07aa48da39140fbe7_6 _overridescale_27_99450af14e8c8ac67e1134b0d87fe85d_5 _spawnsplinemesh_48_a2446a5d40c500f8c62ab69df04cf61c_4 _splinemeshanglethreshold_56_19b857f641a7fd07c727379c3e41af0a_3 _setsplinemeshroll_52_00f53574414c67e7869456be70691314_2) (Utilities|Struct|BreakSTCRule (Variables|/PCGPrimitives/Grammar/BP/BP_ShapeGrammarRule.BP_ShapeGrammarRule_C|GetRule (LocalVariables|Getcurrent%20Rule%20asset))))
  (bind _returnvalue (Utilities|GetObjectPathString _mesh_69_abc2f260431a21daf912f49032642918_2))
  (bind _outmetadata (LocalVariables|GetOutMetadata))
  (bind _outmetadataentry (LocalVariables|GetOutMetaDataEntry))
  (bind _returnvalue_2 (Data|GetParams Input))
  (for _array_element _returnvalue_2
    (LocalVariables|SetPCGData _array_element)
    (Utilities|FlowControl|Sequence
      (:then_0
        ;[1 1 1 1 30] METADATA SETUP
        (bind _returnvalue_3 (Game|ConstructObjectfromClass "/Script/PCG.PCGParamData"))
        (bind _output_get (LocalVariables|SetPCGAttribSet _returnvalue_3))
        (bind _returnvalue_4 (Metadata|MutableMetadata _output_get))
        (LocalVariables|SetOutMetadata _returnvalue_4)
        ;[1 1 1 1 50] SLOT ATTRIBUTES
        (PCG|Metadata|CreateInteger32Attribute (LocalVariables|GetOutMetadata) "Slot ID" -1 false true)
        (PCG|Metadata|CreateSoftObjectPathAttribute (LocalVariables|GetOutMetadata) "Rule_Asset" (LocalVariables|Gettmp%20soft%20object%20path) false true)
        (PCG|Metadata|CreateBoolAttribute (LocalVariables|GetOutMetadata) "Mirror" false false true)
        (PCG|Metadata|CreateInteger32Attribute (LocalVariables|GetOutMetadata) "Type" -1 false true)
        (PCG|Metadata|CreateFloatAttribute (LocalVariables|GetOutMetadata) "Size" -1.0 false true)
        (PCG|Metadata|CreateBoolAttribute (LocalVariables|GetOutMetadata) "Scalable" false false true)
        (PCG|Metadata|CreateVector4Attribute :self (LocalVariables|GetOutMetadata) :AttributeName "Debug Color" :bAllowsInterpolation false :bOverrideParent true)
        (PCG|Metadata|CreateVectorAttribute (LocalVariables|GetOutMetadata) "offset" "0, 0, 0" false true)
        (PCG|Metadata|CreateBoolAttribute (LocalVariables|GetOutMetadata) "is corner" false false true))
      (:then_1
        ;[1 1 1 1 50] RULE ASSET ATTRIBUTES
        (PCG|Metadata|CreateInteger32Attribute (LocalVariables|GetOutMetadata) "Rule ID" -1 false true)
        (PCG|Metadata|CreateFloatAttribute (LocalVariables|GetOutMetadata) "Position On Slot" -1.0 false true)
        (PCG|Metadata|CreateStringAttribute (LocalVariables|GetOutMetadata) "Grammar" "" false true)
        (PCG|Metadata|CreateVectorAttribute (LocalVariables|GetOutMetadata) "Rand_Pos_range" "0, 0, 0" false true)
        (PCG|Metadata|CreateVectorAttribute (LocalVariables|GetOutMetadata) "Rand_Rot_range" "0, 0, 0" false true)
        (PCG|Metadata|CreateVectorAttribute (LocalVariables|GetOutMetadata) "Rand_Scale_range" "0, 0, 0" false true)
        (PCG|Metadata|CreateVectorAttribute (LocalVariables|GetOutMetadata) "Override_Scale" "1.000000,1.000000,1.000000" false true)
        (PCG|Metadata|CreateVectorAttribute (LocalVariables|GetOutMetadata) "Pos_offset" "1.000000,1.000000,1.000000" false true)
        (PCG|Metadata|CreateBoolAttribute (LocalVariables|GetOutMetadata) "Use Own Grammar" false false true)
        (PCG|Metadata|CreateBoolAttribute (LocalVariables|GetOutMetadata) "use up vector" false false true)
        (PCG|Metadata|CreateVectorAttribute (LocalVariables|GetOutMetadata) "up vector" "0, 0, 0" false true)
        (PCG|Metadata|CreateBoolAttribute (LocalVariables|GetOutMetadata) "Spawn Spline Meshes" false false true)
        (PCG|Metadata|CreateFloatAttribute (LocalVariables|GetOutMetadata) "Spline Mesh Angle Threshold" 0.0 false true)
        (PCG|Metadata|CreateBoolAttribute (LocalVariables|GetOutMetadata) "Set Spline Mesh Roll" false false true)
        (PCG|Metadata|CreateFloatAttribute (LocalVariables|GetOutMetadata) "Spline Roll Degrees" 0.0 false true))
      (:then_2
        ;[1 1 1 1 50] GRAMMAR MODULE ATTRIBUTES
        (PCG|Metadata|CreateSoftObjectPathAttribute (LocalVariables|GetOutMetadata) "Module Mesh" (LocalVariables|Gettmp%20soft%20obj%20path) false true)
        (PCG|Metadata|CreateStringAttribute (LocalVariables|GetOutMetadata) "Symbol" "" false true)
        (PCG|Metadata|CreateBoolAttribute (LocalVariables|GetOutMetadata) "GuessSize" false false true)
        (PCG|Metadata|CreateBoolAttribute (LocalVariables|GetOutMetadata) "GuessPivotOffset" false false true)
        (PCG|Metadata|CreateBoolAttribute (LocalVariables|GetOutMetadata) "FlipForwardAxis" false false true)
        (PCG|Metadata|CreateFloatAttribute (LocalVariables|GetOutMetadata) "Module Size" -1.0 false true)
        (PCG|Metadata|CreateVectorAttribute (LocalVariables|GetOutMetadata) "PivotOffset" "0, 0, 0" false true)
        (PCG|Metadata|CreateInteger32Attribute (LocalVariables|GetOutMetadata) "MeshForwardAxis" 0 false true)
        (PCG|Metadata|CreateVector4Attribute :self (LocalVariables|GetOutMetadata) :AttributeName "Module Debug Color" :bAllowsInterpolation false :bOverrideParent true))
      (:then_3
        (bind _aspcgattribute_set (Utilities|Casting|CastToPCGParamData (PCG|TaggedData|BreakTaggedData (LocalVariables|GetPCGData))))
        (bind _returnvalue_5 (Metadata|MutableMetadata _aspcgattribute_set))
        (LocalVariables|SetInMetadata _returnvalue_5)
        (for (_array_element_2 _array_index) (Variables|/PCGPrimitives/Grammar/BP/BP_ShapeGrammarDefinition.BP_ShapeGrammarDefinition_C|GetSlots (Variables|Default|GetShape%20Grammar%20Definition))
          (LocalVariables|SetCurrent%20Slot%20Index _array_index)
          (bind _output_get_2 (LocalVariables|Setcurrent%20Slot%20Config _array_element_2))
          (LocalVariables|SetMirror (.Mirror_40_8AC86C574422496D3787ACAB2ECDD044 (Utilities|Struct|BreakSTCSlot _output_get_2)))
          (LocalVariables|SetCurrent%20Rules (Utilities|Struct|BreakSTCSlot _output_get_2))
          (bind _returnvalue_6 (PCG|Metadata|AddEntry (LocalVariables|GetOutMetadata) -1))
          (LocalVariables|SetOutMetaDataEntry _returnvalue_6)
          (PCG|Metadata|SetInteger32AttributebyMetadataKey (LocalVariables|GetOutMetaDataEntry) (LocalVariables|GetOutMetadata) "Slot ID" (LocalVariables|GetCurrent%20Slot%20Index))
          (PCG|Metadata|SetStringAttributebyMetadataKey (LocalVariables|GetOutMetaDataEntry) (LocalVariables|GetOutMetadata) "Symbol" (Utilities|String|ToString(Name) _symbol_24_e4469dfa4802eef57e43d29349fbc6b7))
          (PCG|Metadata|SetBoolAttributebyMetadataKey (LocalVariables|GetOutMetaDataEntry) (LocalVariables|GetOutMetadata) "Mirror" _mirror_40_8ac86c574422496d3787acab2ecdd044)
          (PCG|Metadata|SetInteger32AttributebyMetadataKey (LocalVariables|GetOutMetaDataEntry) (LocalVariables|GetOutMetadata) "Type")
          (PCG|Metadata|SetFloatAttributebyMetadataKey (LocalVariables|GetOutMetaDataEntry) (LocalVariables|GetOutMetadata) "Size" _size_31_a6850f3346aa84d1d37a079b63b8c073)
          (PCG|Metadata|SetBoolAttributebyMetadataKey (LocalVariables|GetOutMetaDataEntry) (LocalVariables|GetOutMetadata) "scalable" _scalable_35_e032e01047c16dd30ae042a7e8ee0048_3)
          (PCG|Metadata|SetVectorAttributebyMetadataKey (LocalVariables|GetOutMetaDataEntry) (LocalVariables|GetOutMetadata) "offset" _offset_33_5f8333ed4b9d55156e925ab9e6981a83_2)
          (PCG|Metadata|SetBoolAttributebyMetadataKey (LocalVariables|GetOutMetaDataEntry) (LocalVariables|GetOutMetadata) "is corner" _iscorner_37_537b72c44f41655dfbc568871556a218)
          (for (_array_element_3 _array_index_2) (LocalVariables|GetCurrent%20Rules)
            (LocalVariables|Setcurrent%20Rule%20asset _array_element_3)
            (LocalVariables|Setcurrent%20Rule%20index _array_index_2)
            (LocalVariables|Setcurrent%20pos%20on%20Slot (Utilities|Struct|BreakSTCRule (Variables|/PCGPrimitives/Grammar/BP/BP_ShapeGrammarRule.BP_ShapeGrammarRule_C|GetRule _array_element_3)))
            (LocalVariables|Setcurrent%20grammar (.grammar_10_4003928D40C6099DCA57829F48235D57 (Utilities|Struct|BreakSTCRule (Variables|/PCGPrimitives/Grammar/BP/BP_ShapeGrammarRule.BP_ShapeGrammarRule_C|GetRule _array_element_3))))
            (bind _returnvalue_7 (PCG|Metadata|AddEntry (LocalVariables|GetOutMetadata) -1))
            (LocalVariables|SetOutMetaDataEntry _returnvalue_7)
            (PCG|Metadata|SetInteger32AttributebyMetadataKey (LocalVariables|GetOutMetaDataEntry) (LocalVariables|GetOutMetadata) "Slot ID" (LocalVariables|GetCurrent%20Slot%20Index))
            (PCG|Metadata|SetInteger32AttributebyMetadataKey (LocalVariables|GetOutMetaDataEntry) (LocalVariables|GetOutMetadata) "Rule ID" (LocalVariables|Getcurrent%20Rule%20index))
            (PCG|Metadata|SetInteger32AttributebyMetadataKey (LocalVariables|GetOutMetaDataEntry) (LocalVariables|GetOutMetadata) "Type" -1)
            (PCG|Metadata|SetSoftObjectPathAttributebyMetadataKey (LocalVariables|GetOutMetaDataEntry) (LocalVariables|GetOutMetadata) "Rule_Asset" (Utilities|GetSoftObjectPath (LocalVariables|Getcurrent%20Rule%20asset)))
            (PCG|Metadata|SetFloatAttributebyMetadataKey (LocalVariables|GetOutMetaDataEntry) (LocalVariables|GetOutMetadata) "Position On Slot" _positiononslot_57_f123cd704574b53c5d934eaceec6c482_10)
            (PCG|Metadata|SetStringAttributebyMetadataKey (LocalVariables|GetOutMetaDataEntry) (LocalVariables|GetOutMetadata) "Grammar" _grammar_10_4003928d40c6099dca57829f48235d57_9)
            (PCG|Metadata|SetVectorAttributebyMetadataKey (LocalVariables|GetOutMetaDataEntry) (LocalVariables|GetOutMetadata) "Rand_Pos_range" _randposrange_21_4dd3a8d948c5e00d1cd838ad6f83a3df)
            (PCG|Metadata|SetVectorAttributebyMetadataKey (LocalVariables|GetOutMetaDataEntry) (LocalVariables|GetOutMetadata) "Rand_Rot_range" _randrotrange_23_ad11dd064aacbe6ea2fac8ab29b52f76)
            (PCG|Metadata|SetVectorAttributebyMetadataKey (LocalVariables|GetOutMetaDataEntry) (LocalVariables|GetOutMetadata) "Rand_Scale_range" _randscalerange_39_35a8f3d34adce6dd9523b7a68aa529ca_2)
            (PCG|Metadata|SetVectorAttributebyMetadataKey (LocalVariables|GetOutMetaDataEntry) (LocalVariables|GetOutMetadata) "Override_Scale" _overridescale_27_99450af14e8c8ac67e1134b0d87fe85d_4)
            (PCG|Metadata|SetVectorAttributebyMetadataKey (LocalVariables|GetOutMetaDataEntry) (LocalVariables|GetOutMetadata) "Pos_offset" _posoffset_30_a18be791408ce7b07aa48da39140fbe7_5)
            (PCG|Metadata|SetBoolAttributebyMetadataKey (LocalVariables|GetOutMetaDataEntry) (LocalVariables|GetOutMetadata) "Use Own Grammar" _overrideprofilegrammar_36_4ee7ba9446dcf473b7bc3391302e7ef5_3)
            (PCG|Metadata|SetBoolAttributebyMetadataKey (LocalVariables|GetOutMetaDataEntry) (LocalVariables|GetOutMetadata) "use up vector" _useupvector_42_d1e94e694ccc8a3a0238d19ddfb7008b_4)
            (PCG|Metadata|SetVectorAttributebyMetadataKey (LocalVariables|GetOutMetaDataEntry) (LocalVariables|GetOutMetadata) "up vector" _upvector_46_14c9927646621a47516c6eb8732ef3ec_3)
            (PCG|Metadata|SetBoolAttributebyMetadataKey (LocalVariables|GetOutMetaDataEntry) (LocalVariables|GetOutMetadata) "Spawn Spline Meshes" _spawnsplinemesh_48_a2446a5d40c500f8c62ab69df04cf61c_2)
            (PCG|Metadata|SetFloatAttributebyMetadataKey (LocalVariables|GetOutMetaDataEntry) (LocalVariables|GetOutMetadata) "Spline Mesh Angle Threshold" _splinemeshanglethreshold_56_19b857f641a7fd07c727379c3e41af0a_2)
            (PCG|Metadata|SetBoolAttributebyMetadataKey (LocalVariables|GetOutMetaDataEntry) (LocalVariables|GetOutMetadata) "Set Spline Mesh Roll" _setsplinemeshroll_52_00f53574414c67e7869456be70691314_2)
            (PCG|Metadata|SetFloatAttributebyMetadataKey (LocalVariables|GetOutMetaDataEntry) (LocalVariables|GetOutMetadata) "Spline Roll Degrees" _splinerolldegrees_55_a51b57a945178d7632154db6fb7effd4)
            (for _array_element_4 _modulesinfos_38_0ea66fd04214ae8d9f6b83a1b57ac47b_10
              (LocalVariables|SetCurrent%20Grammar%20Module%20def _array_element_4)
              (bind _returnvalue_8 (PCG|Metadata|AddEntry (LocalVariables|GetOutMetadata) -1))
              (LocalVariables|SetOutMetaDataEntry _returnvalue_8)
              (PCG|Metadata|SetInteger32AttributebyMetadataKey (LocalVariables|GetOutMetaDataEntry) (LocalVariables|GetOutMetadata) "Slot ID" (LocalVariables|GetCurrent%20Slot%20Index))
              (PCG|Metadata|SetInteger32AttributebyMetadataKey (LocalVariables|GetOutMetaDataEntry) (LocalVariables|GetOutMetadata) "Rule ID" (LocalVariables|Getcurrent%20Rule%20index))
              (PCG|Metadata|SetInteger32AttributebyMetadataKey (LocalVariables|GetOutMetaDataEntry) (LocalVariables|GetOutMetadata) "Type" -2)
              (PCG|Metadata|SetSoftObjectPathAttributebyMetadataKey (LocalVariables|GetOutMetaDataEntry) (LocalVariables|GetOutMetadata) "Rule_Asset" (Utilities|GetSoftObjectPath (LocalVariables|Getcurrent%20Rule%20asset)))
              (if (Utilities|String|MatchesWildcard _returnvalue "None")
                (PCG|Metadata|SetSoftObjectPathAttributebyMetadataKey _outmetadataentry _outmetadata "Module Mesh" (Utilities|MakeSoftObjectPath (LocalVariables|Gettmp%20string)))
                (else
                  (PCG|Metadata|SetSoftObjectPathAttributebyMetadataKey _outmetadataentry _outmetadata "Module Mesh" (Utilities|MakeSoftObjectPath _returnvalue))))
              (PCG|Metadata|SetStringAttributebyMetadataKey (LocalVariables|GetOutMetaDataEntry) (LocalVariables|GetOutMetadata) "symbol" _symbol_6_f9fb29564dea8bef1ea1dd8aa30c03e3_2)
              (PCG|Metadata|SetBoolAttributebyMetadataKey (LocalVariables|GetOutMetaDataEntry) (LocalVariables|GetOutMetadata) "Scalable" _scalable_68_5f7bc934476923af3c7bbcb6f7239ecf_2)
              (PCG|Metadata|SetBoolAttributebyMetadataKey (LocalVariables|GetOutMetaDataEntry) (LocalVariables|GetOutMetadata) "GuessSize" _guesssize_79_06d919a9451b09f481d71f9db1b48c89_2)
              (PCG|Metadata|SetBoolAttributebyMetadataKey (LocalVariables|GetOutMetaDataEntry) (LocalVariables|GetOutMetadata) "FlipForwardAxis" _flipforwardaxis_30_2ac382c0416cbac8fe032f998f3994de_2)
              (PCG|Metadata|SetBoolAttributebyMetadataKey (LocalVariables|GetOutMetaDataEntry) (LocalVariables|GetOutMetadata) "GuessPivotOffset" _guesspivotoffset_80_dd4df4ef457ae0dcc42e89abe8bbb6c2_3)
              (PCG|Metadata|SetFloatAttributebyMetadataKey (LocalVariables|GetOutMetaDataEntry) (LocalVariables|GetOutMetadata) "Module Size" _size_67_aa4253b1485ed7a89506fca171ac20ba_4)
              (PCG|Metadata|SetVectorAttributebyMetadataKey (LocalVariables|GetOutMetaDataEntry) (LocalVariables|GetOutMetadata) "PivotOffset" _pivotoffset_28_063b0b4741dd9478c20e8e83a4dff20a_3)
              (PCG|Metadata|SetInteger32AttributebyMetadataKey (LocalVariables|GetOutMetaDataEntry) (LocalVariables|GetOutMetadata) "MeshForwardAxis" (Math|Conversions|ToInteger(Byte) _meshforwardaxis_35_62b04dfb4a4ba73267f123a1d53ac25f_3))
              (PCG|Metadata|SetVector4AttributebyMetadataKey (LocalVariables|GetOutMetaDataEntry) (LocalVariables|GetOutMetadata) "Module Debug Color" (Math|Vector4|MakeVector4 (Math|Conversions|ToFloat(Byte) _b) (Math|Conversions|ToFloat(Byte) _g) (Math|Conversions|ToFloat(Byte) _r) (Math|Conversions|ToFloat(Byte) _a)))))))
      (:then_4
        (Utilities|Array|Add (LocalVariables|GetOutData) (PCG|TaggedData|MakeTaggedData (LocalVariables|GetPCGAttribSet))))))
  (return (Utilities|Struct|MakePCGDataCollection (LocalVariables|GetOutData))))

(fn NodeColorOverride
  (return "(R=1.000000,G=0.600000,B=0.000000,A=1.000000)"))

(fn NodeTitleOverride
  (return "Read Shape Grammar Definition"))
