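; Shape grammar PCG actor Blueprint that drives procedural generation from its spline, Size and ShapeGrammarDefinition properties.
; The construction script adds a "MaxHeight:<Size>" component tag to the spline, passing data through tag text.
; It then sets the Size and ShapeGrammarDefinition parameters on the PCG graph instance and calls Generate.
; A separate Cleanup function clears PCG output, and the event stubs are empty.
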(event EventBeginPlay)

(event Collision|EventActorBeginOverlap
  (params (OtherActor :type Actor%20Object%20Reference)))

(event EventTick
  (params (DeltaSeconds :type Float%20%28single-precision%29)))

(fn Cleanup
  (PCG|Cleanup (Variables|Default|GetPCG) true))

(fn Generate
  (PCG|Generate (Variables|Default|GetPCG) false))

(fn ConstructionScript
  (bind _pcg (Variables|Default|GetPCG))
  (bind _graphinstance (Variables|/Script/PCG.PCGComponent|GetGraphInstance _pcg))
  (Utilities|Array|Add (Variables|/Script/Engine.ActorComponent|GetComponentTags (Variables|Default|GetSpline)) (Utilities|String|StringToName (Utilities|String|Append "MaxHeight:" (Utilities|String|ToString(Float) (Variables|Default|GetSize)))))
  (PCG|GraphParameters|SetFloatParameter _graphinstance "Size" (Variables|Default|GetSize))
  (PCG|GraphParameters|SetSoftObjectParameter _graphinstance "ShapeGrammarDefinition" (Utilities|ResolveSoftReference (Variables|Default|GetShapeGrammarDefinition)))
  (CallFunction|Generate))
